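# Post-dynamics Control Rig, by its name, for the face accessory part of the ScrapTunnel medium female player cosmetic.
# Its hierarchy holds the spine, neck, head, clavicles and upper arms, plus four dynamic braid chains under the head, side and back on each side.
# The graph contains only a Begin Execution entry with nothing connected, so the rig modifies no bones.
# A boolean FE variable is declared but unused, and the function library is empty.
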ControlRigBlueprint /BRCosmetics/Characters/Player/Female/Medium/Bodies/F_MED_ScrapTunnel/Meshes/Parts/F_MED_ScrapTunnel_FaceAcc_CtrlRig_PostDyn
graphs [
 RigVMFunctionLibrary {links [], nodes []}
 RigVMModel {nodes [RigUnit_BeginExecution = &RigUnit_BeginExecution()]}
]
hierarchy [
 root {loc=<0 0 0>, rot=<0 0 0 1>}
 pelvis {parent=root, loc=<0 -0.5301 98.0571>, rot=<0.0127 0.707 -0.0127 -0.707>}
 spine_01 {parent=pelvis, loc=<6.1208 -0.9328 0>, rot=<0 0 0.034 -0.9994>}
 spine_02 {parent=spine_01, loc=<6.5997 -0.2095 0>, rot=<0 0 0.0238 -0.9997>}
 spine_03 {parent=spine_02, loc=<6.4914 -0.1237 0>, rot=<0 0 0.0149 -0.9999>}
 spine_04 {parent=spine_03, loc=<6.9598 0.0794 0>, rot=<0 0 -0.0132 -0.9999>}
 spine_05 {parent=spine_04, loc=<6.9987 0 0>, rot=<0 0 -0.0015 -1>}
 neck_01 {parent=spine_05, loc=<11.3138 0.1529 0>, rot=<0 0 -0.2003 -0.9797>}
 neck_02 {parent=neck_01, loc=<4.5627 0 0>, rot=<0 0 0.0504 -0.9987>}
 head {parent=neck_02, loc=<5.3914 0 0>, rot=<0 0 0.115 -0.9934>}
 dyn_braid_side_base_r {parent=head, loc=<0.4172 -4.9658 3.4472>, rot=<0.0041 -1 0 0>}
 dyn_braid_side_01_r {parent=dyn_braid_side_base_r, loc=<0.1903 0 0>, rot=<-0.0204 -0.0971 0.0546 -0.9936>}
 dyn_braid_side_02_r {parent=dyn_braid_side_01_r, loc=<6.5499 0 0>, rot=<-0.1046 0.0649 0.1742 -0.977>}
 dyn_braid_side_03_r {parent=dyn_braid_side_02_r, loc=<8.8708 0 0>, rot=<0.0485 0.0405 -0.0826 -0.9946>}
 dyn_braid_side_04_r {parent=dyn_braid_side_03_r, loc=<10.0848 0 0>, rot=<-0.0028 -0.038 -0.0609 -0.9974>}
 dyn_braid_bk_base_r {parent=head, loc=<1.0672 -6.3222 1.4307>, rot=<0.0041 -1 0 0>}
 dyn_braid_bk_01_r {parent=dyn_braid_bk_base_r, loc=<0.1903 0 0>, rot=<-0.0058 -0.1078 0.0376 -0.9934>}
 dyn_braid_bk_02_r {parent=dyn_braid_bk_01_r, loc=<7.2139 0 0>, rot=<-0.172 0.0875 0.1711 -0.9662>}
 dyn_braid_bk_03_r {parent=dyn_braid_bk_02_r, loc=<8.7426 0 0>, rot=<0.1197 0.0435 -0.0841 -0.9883>}
 dyn_braid_bk_04_r {parent=dyn_braid_bk_03_r, loc=<10.0081 0 0>, rot=<0.0093 0 0 -1>}
 dyn_braid_side_base_l {parent=head, loc=<0.4168 -4.9658 -3.4472>, rot=<0.0041 -1 0 0>}
 dyn_braid_side_01_l {parent=dyn_braid_side_base_l, loc=<0.19 0 0>, rot=<0.0204 0.0971 0.0546 -0.9936>}
 dyn_braid_side_02_l {parent=dyn_braid_side_01_l, loc=<6.5497 0 0>, rot=<0.1046 -0.0649 0.1742 -0.977>}
 dyn_braid_side_03_l {parent=dyn_braid_side_02_l, loc=<8.8708 0.0001 0>, rot=<-0.0485 -0.0405 -0.0826 -0.9946>}
 dyn_braid_side_04_l {parent=dyn_braid_side_03_l, loc=<10.0848 -0.0001 0>, rot=<0.0028 0.038 -0.0609 -0.9974>}
 dyn_braid_bk_base_l {parent=head, loc=<1.0671 -6.3222 -1.4307>, rot=<0.0041 -1 0 0>}
 dyn_braid_bk_01_l {parent=dyn_braid_bk_base_l, loc=<0.19 0 0>, rot=<0.0058 0.1078 0.0376 -0.9934>}
 dyn_braid_bk_02_l {parent=dyn_braid_bk_01_l, loc=<7.214 0 0>, rot=<0.172 -0.0875 0.1711 -0.9662>}
 dyn_braid_bk_03_l {parent=dyn_braid_bk_02_l, loc=<8.7426 0.0001 0>, rot=<-0.1197 -0.0435 -0.0841 -0.9883>}
 dyn_braid_bk_04_l {parent=dyn_braid_bk_03_l, loc=<10.0081 -0.0001 0>, rot=<-0.0093 0 0 -1>}
 clavicle_l {parent=spine_05, loc=<5.4831 -0.1053 -3.2946>, rot=<0.0465 -0.7289 -0.009 -0.6829>}
 upperarm_l {parent=clavicle_l, loc=<9.8658 -1.2136 -0.4105>, rot=<0.0164 -0.3245 -0.0269 -0.9454>}
 clavicle_r {parent=spine_05, loc=<5.4831 -0.1053 3.2946>, rot=<0.7289 0.0465 -0.6829 0.009>}
 upperarm_r {parent=clavicle_r, loc=<-9.8658 1.2136 0.4105>, rot=<0.0164 -0.3245 -0.0269 -0.9454>}
]
variables [FE {type=bool, default=False}]
