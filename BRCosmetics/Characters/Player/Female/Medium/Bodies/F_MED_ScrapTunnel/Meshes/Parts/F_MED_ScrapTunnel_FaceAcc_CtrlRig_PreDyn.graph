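# Pre-dynamics Control Rig for the ScrapTunnel female face accessory, adjusting braid bones from head orientation.
# A spherical pose reader on the head, zeroed when FE is true, drives an inverted, interpolated head_up curve.
# The same unsmoothed weight rotates both side braid roots relative to themselves.
# Four other directional readers execute with unused outputs, and several braid transform nodes are disconnected from execution.
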ControlRigBlueprint /BRCosmetics/Characters/Player/Female/Medium/Bodies/F_MED_ScrapTunnel/Meshes/Parts/F_MED_ScrapTunnel_FaceAcc_CtrlRig_PreDyn
graphs [
 RigVMFunctionLibrary {links [], nodes []}
 RigVMFunction_Sequence_1_ContainedGraph {nodes [
   n1 = RigVMFunctionEntryNode()
   RigVMFunction_Sequence = &RigVMFunction_Sequence(ExecuteContext=@n1.ExecuteContext)
   n2 = &RigVMFunction_Sequence(ExecuteContext=@RigVMFunction_Sequence.B)
   n3 = &RigVMFunction_Sequence(ExecuteContext=@n2.B)
   n4 = &RigVMFunction_Sequence(ExecuteContext=@n3.B)
   RigVMFunctionReturnNode(A=@RigVMFunction_Sequence.A, B=@n2.A, C=@n3.A, D=@n4.A, E=@n4.B)
  ]}
 RigVMModel {nodes [
   RigUnit_BeginExecution = &RigUnit_BeginExecution()
   RigVMFunction_Sequence = RigVMAggregateNode(ExecuteContext=@RigUnit_BeginExecution.ExecutePin)
   n1 = &RigUnit_SphericalPoseReader(
    ActiveRegionScaleFactors=(PositiveWidth=1,NegativeWidth=1,PositiveHeight=2,NegativeHeight=2)
    ActiveRegionSize=0.02
    Debug=(bDrawDebug=True,bDraw2D=False,bDrawLocalAxes=False,DebugScale=25,DebugSegments=20,DebugThickness=0.25)
    DriverAxis=(X=0,Y=1,Z=0)
    DriverItem=(Type=Bone,Name="head")
    FalloffRegionScaleFactors=(PositiveWidth=1,NegativeWidth=1,PositiveHeight=2,NegativeHeight=2)
    FalloffSize=0.12
    FlipHeightScaling=False
    FlipWidthScaling=False
    OptionalParentItem=(Type=Bone,Name="None")
    RotationOffset=(X=90,Y=0,Z=0)
    ExecutePin=@RigVMFunction_Sequence.E
   )
   RigVMFunction_MathBoolNot = &RigVMFunction_MathBoolNot(Value=RigVMVariableNode() {variable=FE}.Value)
   RigVMFunction_MathBoolToFloat = &RigVMFunction_MathBoolToFloat(Value=@RigVMFunction_MathBoolNot.Result)
   Multiply = &RigVMFunction_MathDoubleMul(B=@RigVMFunction_MathBoolToFloat.Result, A=@n1.OutputParam)
   AlphaInterp = &RigVMFunction_AlphaInterp(
    InterpSpeedDecreasing=10
    InterpSpeedIncreasing=10
    OutRange=(Min=1,Max=0)
    Scale=1
    bClampResult=False
    bMapRange=true
    Value=@Multiply.Result
   )
   > /Script/ControlRig.RigUnit_SetCurveValue(Curve=head_up, Value=@AlphaInterp.Result) {is=struct RigVMUnitNode}
   > RigUnit_SetRelativeTransformForItem = /Script/ControlRig.RigUnit_SetRelativeTransformForItem(
    Child=(Type=Bone,Name="dyn_braid_side_01_r")
    Parent=(Type=Bone,Name="dyn_braid_side_01_r")
    Value=(Rotation=(X=0,Y=0,Z=-0.1305,W=0.9914),Translation=(X=0,Y=0,Z=0),Scale3D=(X=1,Y=1,Z=1))
    bParentInitial=False
    bPropagateToChildren=false
    Weight=@Multiply.Result
   ) {
    is=struct RigVMUnitNode
    template="Set Relative Transform::Execute(in Child,in Parent,in bParentInitial,in Value,in Weight,in bPropagateToChildren)"
   }
   > &RigUnit_SetRelativeTransformForItem(
    Child=(Type=Bone,Name="dyn_braid_side_01_l")
    Parent=(Type=Bone,Name="dyn_braid_side_01_l")
    Value=(Rotation=(X=0,Y=0,Z=-0.1305,W=0.9914),Translation=(X=0,Y=0,Z=0),Scale3D=(X=1,Y=1,Z=1))
    bParentInitial=False
    bPropagateToChildren=false
    Weight=@Multiply.Result
   ) {template="Set Relative Transform::Execute(in Child,in Parent,in bParentInitial,in Value,in Weight,in bPropagateToChildren)"}
   &RigUnit_SphericalPoseReader(
    ActiveRegionScaleFactors=(PositiveWidth=0.75,NegativeWidth=0.75,PositiveHeight=0.5,NegativeHeight=1)
    ActiveRegionSize=0.1
    Debug=(bDrawDebug=True,bDraw2D=False,bDrawLocalAxes=False,DebugScale=25,DebugSegments=20,DebugThickness=0.25)
    DriverAxis=(X=0,Y=1,Z=0)
    DriverItem=(Type=Bone,Name="head")
    FalloffRegionScaleFactors=(PositiveWidth=1,NegativeWidth=1,PositiveHeight=0.5,NegativeHeight=1)
    FalloffSize=0.2
    FlipHeightScaling=False
    FlipWidthScaling=False
    OptionalParentItem=(Type=Bone,Name="None")
    RotationOffset=(X=60,Y=0,Z=0)
    ExecutePin=@RigVMFunction_Sequence.D
   )
   &RigUnit_SphericalPoseReader(
    ActiveRegionScaleFactors=(PositiveWidth=1,NegativeWidth=0.5,PositiveHeight=0.75,NegativeHeight=0.75)
    ActiveRegionSize=0.1
    Debug=(bDrawDebug=True,bDraw2D=False,bDrawLocalAxes=False,DebugScale=25,DebugSegments=20,DebugThickness=0.25)
    DriverAxis=(X=0,Y=1,Z=0)
    DriverItem=(Type=Bone,Name="head")
    FalloffRegionScaleFactors=(PositiveWidth=1,NegativeWidth=0.5,PositiveHeight=1,NegativeHeight=1)
    FalloffSize=0.2
    FlipHeightScaling=False
    FlipWidthScaling=False
    OptionalParentItem=(Type=Bone,Name="None")
    RotationOffset=(X=90,Y=0,Z=-30)
    ExecutePin=@RigVMFunction_Sequence.A
   )
   &RigUnit_SphericalPoseReader(
    ActiveRegionScaleFactors=(PositiveWidth=0.5,NegativeWidth=1,PositiveHeight=0.75,NegativeHeight=0.75)
    ActiveRegionSize=0.1
    Debug=(bDrawDebug=True,bDraw2D=False,bDrawLocalAxes=False,DebugScale=25,DebugSegments=20,DebugThickness=0.25)
    DriverAxis=(X=0,Y=1,Z=0)
    DriverItem=(Type=Bone,Name="head")
    FalloffRegionScaleFactors=(PositiveWidth=0.5,NegativeWidth=1,PositiveHeight=1,NegativeHeight=1)
    FalloffSize=0.2
    FlipHeightScaling=False
    FlipWidthScaling=False
    OptionalParentItem=(Type=Bone,Name="None")
    RotationOffset=(X=90,Y=0,Z=30)
    ExecutePin=@RigVMFunction_Sequence.B
   )
   &RigUnit_SphericalPoseReader(
    ActiveRegionScaleFactors=(PositiveWidth=0.75,NegativeWidth=0.75,PositiveHeight=1,NegativeHeight=0.5)
    ActiveRegionSize=0.1
    Debug=(bDrawDebug=True,bDraw2D=False,bDrawLocalAxes=False,DebugScale=25,DebugSegments=20,DebugThickness=0.25)
    DriverAxis=(X=0,Y=1,Z=0)
    DriverItem=(Type=Bone,Name="head")
    FalloffRegionScaleFactors=(PositiveWidth=1,NegativeWidth=1,PositiveHeight=1,NegativeHeight=0.5)
    FalloffSize=0.2
    FlipHeightScaling=False
    FlipWidthScaling=False
    OptionalParentItem=(Type=Bone,Name="None")
    RotationOffset=(X=120,Y=0,Z=0)
    ExecutePin=@RigVMFunction_Sequence.C
   )
   RerouteNode_Constant = RigVMDispatchNode(Value=1) {template="DISPATCH_RigVMDispatch_Constant(io Value)"}
   &RigUnit_SetRelativeTransformForItem(
    Child=(Type=Bone,Name="dyn_braid_side_01_r")
    ExecutePin=()
    Parent=(Type=Bone,Name="dyn_braid_side_base_r")
    Value=(Rotation=(X=0,Y=0,Z=0.8192,W=-0.5736),Translation=(X=0,Y=0,Z=0),Scale3D=(X=1,Y=1,Z=1))
    bParentInitial=False
    bPropagateToChildren=false
    Weight=@RerouteNode_Constant.Value
   ) {template="Set Relative Transform::Execute(in Child,in Parent,in bParentInitial,in Value,in Weight,in bPropagateToChildren)"}
   > &RigUnit_SetRelativeTransformForItem(
    Child=(Type=Bone,Name="dyn_braid_bk_01_r")
    Parent=(Type=Bone,Name="dyn_braid_bk_base_r")
    Value=(Rotation=(X=0.0606,Y=0.0619,Z=0.7045,W=-0.7044),Translation=(X=0,Y=0,Z=0),Scale3D=(X=1,Y=1,Z=1))
    bParentInitial=False
    bPropagateToChildren=false
    Weight=@RerouteNode_Constant.Value
   ) {template="Set Relative Transform::Execute(in Child,in Parent,in bParentInitial,in Value,in Weight,in bPropagateToChildren)"}
   > &RigUnit_SetRelativeTransformForItem(
    Child=(Type=Bone,Name="dyn_braid_bk_01_l")
    Parent=(Type=Bone,Name="dyn_braid_bk_base_l")
    Value=(Rotation=(X=0,Y=0,Z=-0.7072,W=0.707),Translation=(X=0,Y=0,Z=0),Scale3D=(X=1,Y=1,Z=1))
    bParentInitial=False
    bPropagateToChildren=false
    Weight=@RerouteNode_Constant.Value
   ) {template="Set Relative Transform::Execute(in Child,in Parent,in bParentInitial,in Value,in Weight,in bPropagateToChildren)"}
   > &RigUnit_SetRelativeTransformForItem(
    Child=(Type=Bone,Name="dyn_braid_side_01_l")
    Parent=(Type=Bone,Name="dyn_braid_side_base_l")
    Value=(Rotation=(X=0,Y=0,Z=-0.7072,W=0.707),Translation=(X=0,Y=0,Z=0),Scale3D=(X=1,Y=1,Z=1))
    bParentInitial=False
    bPropagateToChildren=false
    Weight=@RerouteNode_Constant.Value
   ) {template="Set Relative Transform::Execute(in Child,in Parent,in bParentInitial,in Value,in Weight,in bPropagateToChildren)"}
   &RigUnit_ModifyTransforms(
    ExecutePin=()
    ItemToModify=((Item=(Type=Bone,Name="dyn_braid_side_01_r"),Transform=(Rotation=(X=0,Y=0,Z=-0.2588,W=0.9659),Translation=(X=0,Y=0,Z=0),Scale3D=(X=1,Y=1,Z=1))))
    Weight=1
    WeightMinimum=0
   )
   &RigUnit_OffsetTransformForItem(
    ExecutePin=()
    Item=(Type=Bone,Name="dyn_braid_side_01_r")
    OffsetTransform=(Rotation=(X=0,Y=0,Z=-0.0872,W=0.9962),Translation=(X=0,Y=0,Z=0),Scale3D=(X=1,Y=1,Z=1))
    Weight=1
    bPropagateToChildren=false
   )
  ]}
]
hierarchy [
 root {loc=<0 0 0>, rot=<0 0 0 1>}
 pelvis {parent=root, loc=<0 -0.5301 98.0571>, rot=<0.0127 0.707 -0.0127 -0.707>}
 spine_01 {parent=pelvis, loc=<6.1208 -0.9328 0>, rot=<0 0 0.034 -0.9994>}
 spine_02 {parent=spine_01, loc=<6.5997 -0.2095 0>, rot=<0 0 0.0238 -0.9997>}
 spine_03 {parent=spine_02, loc=<6.4914 -0.1237 0>, rot=<0 0 0.0149 -0.9999>}
 spine_04 {parent=spine_03, loc=<6.9598 0.0794 0>, rot=<0 0 -0.0132 -0.9999>}
 spine_05 {parent=spine_04, loc=<6.9987 0 0>, rot=<0 0 -0.0015 -1>}
 neck_01 {parent=spine_05, loc=<11.3138 0.1529 0>, rot=<0 0 -0.2003 -0.9797>}
 neck_02 {parent=neck_01, loc=<4.5627 0 0>, rot=<0 0 0.0504 -0.9987>}
 head {parent=neck_02, loc=<5.3914 0 0>, rot=<0 0 0.115 -0.9934>}
 dyn_braid_side_base_r {parent=head, loc=<0.4172 -4.9658 3.4472>, rot=<0.0041 -1 0 0>}
 dyn_braid_side_01_r {parent=dyn_braid_side_base_r, loc=<0.1903 0 0>, rot=<-0.0204 -0.0971 0.0546 -0.9936>}
 dyn_braid_side_02_r {parent=dyn_braid_side_01_r, loc=<6.5499 0 0>, rot=<-0.1046 0.0649 0.1742 -0.977>}
 dyn_braid_side_03_r {parent=dyn_braid_side_02_r, loc=<8.8708 0 0>, rot=<0.0485 0.0405 -0.0826 -0.9946>}
 dyn_braid_side_04_r {parent=dyn_braid_side_03_r, loc=<10.0848 0 0>, rot=<-0.0028 -0.038 -0.0609 -0.9974>}
 dyn_braid_bk_base_r {parent=head, loc=<1.0672 -6.3222 1.4307>, rot=<0.0041 -1 0 0>}
 dyn_braid_bk_01_r {parent=dyn_braid_bk_base_r, loc=<0.1903 0 0>, rot=<-0.0058 -0.1078 0.0376 -0.9934>}
 dyn_braid_bk_02_r {parent=dyn_braid_bk_01_r, loc=<7.2139 0 0>, rot=<-0.172 0.0875 0.1711 -0.9662>}
 dyn_braid_bk_03_r {parent=dyn_braid_bk_02_r, loc=<8.7426 0 0>, rot=<0.1197 0.0435 -0.0841 -0.9883>}
 dyn_braid_bk_04_r {parent=dyn_braid_bk_03_r, loc=<10.0081 0 0>, rot=<0.0093 0 0 -1>}
 dyn_braid_side_base_l {parent=head, loc=<0.4168 -4.9658 -3.4472>, rot=<0.0041 -1 0 0>}
 dyn_braid_side_01_l {parent=dyn_braid_side_base_l, loc=<0.19 0 0>, rot=<0.0204 0.0971 0.0546 -0.9936>}
 dyn_braid_side_02_l {parent=dyn_braid_side_01_l, loc=<6.5497 0 0>, rot=<0.1046 -0.0649 0.1742 -0.977>}
 dyn_braid_side_03_l {parent=dyn_braid_side_02_l, loc=<8.8708 0.0001 0>, rot=<-0.0485 -0.0405 -0.0826 -0.9946>}
 dyn_braid_side_04_l {parent=dyn_braid_side_03_l, loc=<10.0848 -0.0001 0>, rot=<0.0028 0.038 -0.0609 -0.9974>}
 dyn_braid_bk_base_l {parent=head, loc=<1.0671 -6.3222 -1.4307>, rot=<0.0041 -1 0 0>}
 dyn_braid_bk_01_l {parent=dyn_braid_bk_base_l, loc=<0.19 0 0>, rot=<0.0058 0.1078 0.0376 -0.9934>}
 dyn_braid_bk_02_l {parent=dyn_braid_bk_01_l, loc=<7.214 0 0>, rot=<0.172 -0.0875 0.1711 -0.9662>}
 dyn_braid_bk_03_l {parent=dyn_braid_bk_02_l, loc=<8.7426 0.0001 0>, rot=<-0.1197 -0.0435 -0.0841 -0.9883>}
 dyn_braid_bk_04_l {parent=dyn_braid_bk_03_l, loc=<10.0081 -0.0001 0>, rot=<-0.0093 0 0 -1>}
 clavicle_l {parent=spine_05, loc=<5.4831 -0.1053 -3.2946>, rot=<0.0465 -0.7289 -0.009 -0.6829>}
 upperarm_l {parent=clavicle_l, loc=<9.8658 -1.2136 -0.4105>, rot=<0.0164 -0.3245 -0.0269 -0.9454>}
 clavicle_r {parent=spine_05, loc=<5.4831 -0.1053 3.2946>, rot=<0.7289 0.0465 -0.6829 0.009>}
 upperarm_r {parent=clavicle_r, loc=<-9.8658 1.2136 0.4105>, rot=<0.0164 -0.3245 -0.0269 -0.9454>}
 head_up {type=Curve, loc=<0 0 0>, rot=<0 0 0 1>}
]
variables [FE {type=bool, default=False}]
