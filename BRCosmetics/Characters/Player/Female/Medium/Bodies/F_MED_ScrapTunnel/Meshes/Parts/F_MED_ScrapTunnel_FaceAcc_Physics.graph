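# Physics asset for the face accessory part of the female medium ScrapTunnel player cosmetic, covering four dynamic braids: back and side, left and right.
# Each braid is a capsule chain from the head, mostly with limited swing and tight twist, though some right-side upper joints are free.
# Spine, neck, head, clavicle and upper-arm capsules supply body volumes, with free constraints from spine_05 upward.
# Extra free constraints cross-link the four braid tips.
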PhysicsAsset /BRCosmetics/Characters/Player/Female/Medium/Bodies/F_MED_ScrapTunnel/Meshes/Parts/F_MED_ScrapTunnel_FaceAcc_Physics
bodies {
 clavicle_l [{shapeType=Capsule, length=10.6, radius=6.8, shapeName=clavicle_l_capsule, center=<6.2526 0.1462 -2.2573>, rotation=<rot 0 -90 90>}]
 clavicle_r [
  {shapeType=Capsule, length=10.6, radius=8, shapeName=clavicle_l_capsule, center=<-3.1925 -1.5071 2.2573>, rotation=<rot 0 -90 90>}
  {
   shapeType=Capsule
   length=14
   radius=6.8
   shapeName=clavicle_r_capsule
   center=<-3.6715 -10.977 -14.1227>
   rotation=<rot 24.5939 -94.6292 -11.0095>
  }
 ]
 dyn_braid_bk_01_l [{shapeType=Capsule, length=6, radius=2, shapeName=dyn_braid_bk_01_l_capsule, center=<6 0 0>, rotation=<rot 90 0 0>}]
 dyn_braid_bk_01_r [{shapeType=Capsule, length=6, radius=3, shapeName=dyn_braid_bk_01_r_capsule, center=<6 0 0>, rotation=<rot 90 0 0>}]
 dyn_braid_bk_02_l [{shapeType=Capsule, length=4, radius=2, shapeName=dyn_braid_bk_02_l_capsule, center=<4 0 0>, rotation=<rot 90 0 0>}]
 dyn_braid_bk_02_r [{shapeType=Capsule, length=4, radius=2, shapeName=dyn_braid_bk_02_r_capsule, center=<4 0 0>, rotation=<rot 90 0 0>}]
 dyn_braid_bk_03_l [{shapeType=Capsule, length=4, radius=2, shapeName=dyn_braid_bk_03_l_capsule, center=<4 0 0>, rotation=<rot 90 0 0>}]
 dyn_braid_bk_03_r [{shapeType=Capsule, length=4, radius=2, shapeName=dyn_braid_bk_03_r_capsule, center=<4 0 0>, rotation=<rot 90 0 0>}]
 dyn_braid_bk_04_l [{shapeType=Capsule, length=4, radius=2, shapeName=dyn_braid_bk_04_l_capsule, center=<4 0 0>, rotation=<rot 90 0 0>}]
 dyn_braid_bk_04_r [{shapeType=Capsule, length=4, radius=2, shapeName=dyn_braid_bk_04_r_capsule, center=<4 0 0>, rotation=<rot 90 0 0>}]
 dyn_braid_bk_base_l [{shapeType=Capsule, length=2, radius=2, shapeName=dyn_braid_bk_base_l_capsule}]
 dyn_braid_bk_base_r [{shapeType=Capsule, length=2, radius=2, shapeName=dyn_braid_bk_base_r_capsule}]
 dyn_braid_side_01_l [{shapeType=Capsule, length=6, radius=2, shapeName=dyn_braid_side_01_l_capsule, center=<6 0 0>, rotation=<rot 90 0 0>}]
 dyn_braid_side_01_r [{shapeType=Capsule, length=6, radius=3, shapeName=dyn_braid_side_01_r_capsule, center=<6 0 0>, rotation=<rot 90 0 0>}]
 dyn_braid_side_02_l [{shapeType=Capsule, length=4, radius=2, shapeName=dyn_braid_side_02_l_capsule, center=<4 0 0>, rotation=<rot 90 0 0>}]
 dyn_braid_side_02_r [{shapeType=Capsule, length=4, radius=2, shapeName=dyn_braid_side_02_r_capsule, center=<4 0 0>, rotation=<rot 90 0 0>}]
 dyn_braid_side_03_l [{shapeType=Capsule, length=4, radius=2, shapeName=dyn_braid_side_03_l_capsule, center=<4 0 0>, rotation=<rot 90 0 0>}]
 dyn_braid_side_03_r [{shapeType=Capsule, length=4, radius=2, shapeName=dyn_braid_side_03_r_capsule, center=<4 0 0>, rotation=<rot 90 0 0>}]
 dyn_braid_side_04_l [{shapeType=Capsule, length=4, radius=2, shapeName=dyn_braid_side_04_l_capsule, center=<4 0 0>, rotation=<rot 90 0 0>}]
 dyn_braid_side_04_r [{shapeType=Capsule, length=4, radius=2, shapeName=dyn_braid_side_04_r_capsule, center=<4 0 0>, rotation=<rot 90 0 0>}]
 dyn_braid_side_base_l [{shapeType=Capsule, length=2, radius=2, shapeName=dyn_braid_side_base_l_capsule}]
 dyn_braid_side_base_r [{shapeType=Capsule, length=2, radius=2, shapeName=dyn_braid_side_base_r_capsule}]
 head [{shapeType=Capsule, length=7.8577, radius=6.2891, center=<0 2.7846 0>, rotation=<rot -90 180 179.9998>}]
 neck_01 [{shapeType=Capsule, length=0, radius=3, shapeName=neck_01_capsule, center=<1.2134 0.1918 0>, rotation=<rot 90 0 0>}]
 neck_02 [{shapeType=Capsule, length=28.8, radius=15, shapeName=spine_05_capsule, center=<-8.4615 12.9244 0>}]
 spine_01 [{shapeType=Capsule, length=20, radius=12, shapeName=spine_01_capsule, rotation=<rot 0 -90 0>}]
 spine_02 [{shapeType=Capsule, length=15.7063, radius=10.2937, shapeName=spine_02_capsule, center=<6.0567 1.2153 0>}]
 spine_03 [{shapeType=Capsule, length=15.9887, radius=10.0113, shapeName=spine_02_capsule, center=<6.0567 2.4426 0>}]
 spine_04 [
  {shapeType=Capsule, length=17.8, radius=9.1, shapeName=spine_04_capsule, center=<4.0135 1.6435 0>, rotation=<rot 0 -90 0>}
  {shapeType=Capsule, length=20.8, radius=6.1, shapeName=spine_04_capsule, center=<14.0135 1.8314 0>, rotation=<rot 0 -90 0>}
 ]
 spine_05 [
  {shapeType=Capsule, length=17.2, radius=8, shapeName=spine_05_capsule, center=<2.8857 2.8036 0>, rotation=<rot 0 -90 0>}
  {shapeType=Capsule, length=4, radius=10, shapeName=spine_05_capsule, center=<10.0796 6.5982 0>, rotation=<rot 0 -90 0>}
  {shapeType=Capsule, length=10, radius=8, shapeName=spine_05_capsule, center=<5.4 2.3643 0>, rotation=<rot 0 -90 0>}
  {shapeType=Capsule, length=10, radius=8, shapeName=spine_05_capsule, center=<-1.8862 1.4949 0>, rotation=<rot 0 -90 0>}
 ]
 upperarm_l [{shapeType=Capsule, length=13.5146, radius=7, shapeName=upperarm_l_capsule, center=<11.2314 -0.3627 0>, rotation=<rot 90 0 0>}]
 upperarm_r [{shapeType=Capsule, length=13.5146, radius=6, shapeName=upperarm_l_capsule, center=<-11.2314 0.3627 0>, rotation=<rot 90 0 0>}]
}
constraints [
 {
  bone1Name=clavicle_l
  bone2Name=spine_05
  swing1LimitDegrees=45
  swing1Motion=Free
  swing2LimitDegrees=45
  swing2Motion=Free
  twistLimitDegrees=45
  twistMotion=Free
 }
 {
  bone1Name=upperarm_l
  bone2Name=clavicle_l
  swing1LimitDegrees=45
  swing1Motion=Free
  swing2LimitDegrees=45
  swing2Motion=Free
  twistLimitDegrees=45
  twistMotion=Free
 }
 {
  bone1Name=neck_01
  bone2Name=spine_05
  swing1LimitDegrees=45
  swing1Motion=Free
  swing2LimitDegrees=45
  swing2Motion=Free
  twistLimitDegrees=45
  twistMotion=Free
 }
 {
  bone1Name=neck_02
  bone2Name=neck_01
  swing1LimitDegrees=45
  swing1Motion=Free
  swing2LimitDegrees=45
  swing2Motion=Free
  twistLimitDegrees=45
  twistMotion=Free
 }
 {
  bone1Name=head
  bone2Name=neck_02
  swing1LimitDegrees=45
  swing1Motion=Free
  swing2LimitDegrees=45
  swing2Motion=Free
  twistLimitDegrees=45
  twistMotion=Free
 }
 {
  bone1Name=clavicle_r
  bone2Name=spine_05
  swing1LimitDegrees=45
  swing1Motion=Free
  swing2LimitDegrees=45
  swing2Motion=Free
  twistLimitDegrees=45
  twistMotion=Free
 }
 {
  bone1Name=upperarm_r
  bone2Name=clavicle_r
  swing1LimitDegrees=45
  swing1Motion=Free
  swing2LimitDegrees=45
  swing2Motion=Free
  twistLimitDegrees=45
  twistMotion=Free
 }
 {bone1Name=dyn_braid_side_04_r, bone2Name=dyn_braid_side_03_r, swing1LimitDegrees=45, swing2LimitDegrees=45, twistLimitDegrees=5}
 {bone1Name=dyn_braid_bk_base_l, bone2Name=head, swing1LimitDegrees=45, swing2LimitDegrees=45, twistLimitDegrees=5}
 {bone1Name=dyn_braid_bk_01_l, bone2Name=dyn_braid_bk_base_l, swing1LimitDegrees=45, swing2LimitDegrees=45, twistLimitDegrees=5}
 {bone1Name=dyn_braid_bk_02_l, bone2Name=dyn_braid_bk_01_l, swing1LimitDegrees=60, swing2LimitDegrees=45, twistLimitDegrees=5}
 {bone1Name=dyn_braid_bk_03_l, bone2Name=dyn_braid_bk_02_l, swing1LimitDegrees=30, swing2LimitDegrees=45, twistLimitDegrees=5}
 {bone1Name=dyn_braid_bk_04_l, bone2Name=dyn_braid_bk_03_l, swing1LimitDegrees=45, swing2LimitDegrees=45, twistLimitDegrees=5}
 {
  bone1Name=dyn_braid_bk_base_r
  bone2Name=head
  swing1LimitDegrees=45
  swing1Motion=Free
  swing2LimitDegrees=45
  swing2Motion=Free
  twistLimitDegrees=5
  twistMotion=Free
 }
 {bone1Name=dyn_braid_bk_01_r, bone2Name=dyn_braid_bk_base_r, swing1LimitDegrees=45, swing2LimitDegrees=30, twistLimitDegrees=5}
 {bone1Name=dyn_braid_bk_02_r, bone2Name=dyn_braid_bk_01_r, swing1LimitDegrees=60, swing2LimitDegrees=45, twistLimitDegrees=5}
 {bone1Name=dyn_braid_bk_03_r, bone2Name=dyn_braid_bk_02_r, swing1LimitDegrees=30, swing2LimitDegrees=45, twistLimitDegrees=5}
 {bone1Name=dyn_braid_bk_04_r, bone2Name=dyn_braid_bk_03_r, swing1LimitDegrees=45, swing2LimitDegrees=45, twistLimitDegrees=5}
 {bone1Name=dyn_braid_side_base_l, bone2Name=head, swing1LimitDegrees=45, swing2LimitDegrees=45, twistLimitDegrees=5}
 {bone1Name=dyn_braid_side_01_l, bone2Name=dyn_braid_side_base_l, swing1LimitDegrees=45, swing2LimitDegrees=45, twistLimitDegrees=5}
 {bone1Name=dyn_braid_side_02_l, bone2Name=dyn_braid_side_01_l, swing1LimitDegrees=60, swing2LimitDegrees=45, twistLimitDegrees=5}
 {bone1Name=dyn_braid_side_03_l, bone2Name=dyn_braid_side_02_l, swing1LimitDegrees=30, swing2LimitDegrees=45, twistLimitDegrees=5}
 {bone1Name=dyn_braid_side_04_l, bone2Name=dyn_braid_side_03_l, swing1LimitDegrees=45, swing2LimitDegrees=45, twistLimitDegrees=5}
 {
  bone1Name=dyn_braid_side_base_r
  bone2Name=head
  swing1LimitDegrees=45
  swing1Motion=Free
  swing2LimitDegrees=45
  swing2Motion=Free
  twistLimitDegrees=5
  twistMotion=Free
 }
 {
  bone1Name=dyn_braid_side_01_r
  bone2Name=dyn_braid_side_base_r
  swing1LimitDegrees=60
  swing1Motion=Free
  swing2LimitDegrees=60
  swing2Motion=Free
  twistLimitDegrees=5
  twistMotion=Free
 }
 {bone1Name=dyn_braid_side_02_r, bone2Name=dyn_braid_side_01_r, swing1LimitDegrees=60, swing2LimitDegrees=45, twistLimitDegrees=5}
 {bone1Name=dyn_braid_side_03_r, bone2Name=dyn_braid_side_02_r, swing1LimitDegrees=30, swing2LimitDegrees=45, twistLimitDegrees=5}
 {
  bone1Name=dyn_braid_bk_04_l
  bone2Name=dyn_braid_side_04_l
  swing1LimitDegrees=45
  swing1Motion=Free
  swing2LimitDegrees=45
  swing2Motion=Free
  twistLimitDegrees=45
  twistMotion=Free
 }
 {
  bone1Name=dyn_braid_bk_04_r
  bone2Name=dyn_braid_bk_04_l
  swing1LimitDegrees=45
  swing1Motion=Free
  swing2LimitDegrees=45
  swing2Motion=Free
  twistLimitDegrees=45
  twistMotion=Free
 }
 {
  bone1Name=dyn_braid_bk_04_r
  bone2Name=dyn_braid_side_04_r
  swing1LimitDegrees=45
  swing1Motion=Free
  swing2LimitDegrees=45
  swing2Motion=Free
  twistLimitDegrees=45
  twistMotion=Free
 }
]
