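# Post-process style AnimBlueprint for the F_MED_ScrapTunnel face accessory, adding secondary motion to the character's four braids.
# Copied parent and body poses pass through a PreDyn Control Rig into a cache feeding RigidBody branches for default, skydive, backpack, cape and front end.
# Bool blends select branches, with constraints correcting braids for backpack/cape, before a PostDyn Control Rig.
# EventGraph derives state flags, gravity overrides and noise-driven tassel rotations.
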AnimBlueprint /BRCosmetics/Characters/Player/Female/Medium/Bodies/F_MED_ScrapTunnel/Meshes/Parts/F_MED_ScrapTunnel_FaceAcc_F_MED_ScrapTunnel_FaceAcc_Anim
target_skeleton=&/F_MED_ScrapTunnel_FaceAcc_Skeleton
graphs [
 AnimGraph {nodes [
   n1 = "Constraint - dyn_braid_side_02_l (PPPPPPP)"!(
    ComponentPose
    Alpha=1
    ConstraintWeights_0=0.3194
    ConstraintWeights_1:float=0.2484
    ConstraintWeights_2:float=0.2229
    ConstraintWeights_3:float=0.1248
    ConstraintWeights_4:float=0.0651
    ConstraintWeights_5:float=0.0183
    ConstraintWeights_6:float=0.0011
   ) -> (Pose) {
    alphaBoolBlend {blendOption=Linear}
    boneToModify {boneName=dyn_braid_side_02_l}
    constraintSetup [
     {targetBone {boneName=neck_01}}
     {targetBone {boneName=spine_05}}
     {targetBone {boneName=neck_02}}
     {targetBone {boneName=head}}
     {targetBone {boneName=clavicle_l}}
     {targetBone {boneName=spine_04}}
     {targetBone {boneName=clavicle_r}}
    ]
    constraintWeights [1,1,1,1,1,1,1]
   }
   n2 = "Constraint - dyn_braid_side_03_l (PPPPP)"!(
    ComponentPose=@n1
    Alpha=1
    ConstraintWeights_0=0.4756
    ConstraintWeights_1:float=0.4188
    ConstraintWeights_2:float=0.0578
    ConstraintWeights_3:float=0.0396
    ConstraintWeights_4:float=0.0082
   ) -> (Pose) {
    alphaBoolBlend {blendOption=Linear}
    boneToModify {boneName=dyn_braid_side_03_l}
    constraintSetup [
     {targetBone {boneName=clavicle_l}}
     {targetBone {boneName=spine_05}}
     {targetBone {boneName=clavicle_r}}
     {targetBone {boneName=spine_04}}
     {targetBone {boneName=spine_03}}
    ]
    constraintWeights [1,1,1,1,1]
   }
   n3 = "Constraint - dyn_braid_side_04_l (PPPPPP)"!(
    ComponentPose=@n2
    Alpha=1
    ConstraintWeights_0=0.428
    ConstraintWeights_1:float=0.2517
    ConstraintWeights_2:float=0.1334
    ConstraintWeights_3:float=0.0789
    ConstraintWeights_4:float=0.0699
    ConstraintWeights_5:float=0.038
   ) -> (Pose) {
    alphaBoolBlend {blendOption=Linear}
    boneToModify {boneName=dyn_braid_side_04_l}
    constraintSetup [
     {targetBone {boneName=spine_05}}
     {targetBone {boneName=spine_04}}
     {targetBone {boneName=clavicle_l}}
     {targetBone {boneName=clavicle_r}}
     {targetBone {boneName=spine_02}}
     {targetBone {boneName=spine_03}}
    ]
    constraintWeights [1,1,1,1,1,1]
   }
   n4 = "Constraint - dyn_braid_bk_02_l (PPPPPPP)"!(
    ComponentPose=@n3
    Alpha=1
    ConstraintWeights_0=0.3803
    ConstraintWeights_1:float=0.2329
    ConstraintWeights_2:float=0.2181
    ConstraintWeights_3:float=0.1151
    ConstraintWeights_4:float=0.0284
    ConstraintWeights_5:float=0.0225
    ConstraintWeights_6:float=0.0028
   ) -> (Pose) {
    alphaBoolBlend {blendOption=Linear}
    boneToModify {boneName=dyn_braid_bk_02_l}
    constraintSetup [
     {targetBone {boneName=neck_01}}
     {targetBone {boneName=neck_02}}
     {targetBone {boneName=spine_05}}
     {targetBone {boneName=head}}
     {targetBone {boneName=clavicle_l}}
     {targetBone {boneName=spine_04}}
     {targetBone {boneName=clavicle_r}}
    ]
    constraintWeights [1,1,1,1,1,1,1]
   }
   n5 = "Constraint - dyn_braid_bk_03_l (PPPPPP)"!(
    ComponentPose=@n4
    Alpha=1
    ConstraintWeights_0=0.5271
    ConstraintWeights_1:float=0.3273
    ConstraintWeights_2:float=0.0954
    ConstraintWeights_3:float=0.0387
    ConstraintWeights_4:float=0.0104
    ConstraintWeights_5:float=0.001
   ) -> (Pose) {
    alphaBoolBlend {blendOption=Linear}
    boneToModify {boneName=dyn_braid_bk_03_l}
    constraintSetup [
     {targetBone {boneName=spine_05}}
     {targetBone {boneName=clavicle_l}}
     {targetBone {boneName=clavicle_r}}
     {targetBone {boneName=spine_04}}
     {targetBone {boneName=spine_03}}
     {targetBone {boneName=spine_02}}
    ]
    constraintWeights [1,1,1,1,1,1]
   }
   n6 = "Constraint - dyn_braid_bk_04_l (PPPPPP)"!(
    ComponentPose=@n5
    Alpha=1
    ConstraintWeights_0=0.4229
    ConstraintWeights_1:float=0.2431
    ConstraintWeights_2:float=0.1295
    ConstraintWeights_3:float=0.1029
    ConstraintWeights_4:float=0.0642
    ConstraintWeights_5:float=0.0373
   ) -> (Pose) {
    alphaBoolBlend {blendOption=Linear}
    boneToModify {boneName=dyn_braid_bk_04_l}
    constraintSetup [
     {targetBone {boneName=spine_05}}
     {targetBone {boneName=spine_04}}
     {targetBone {boneName=clavicle_l}}
     {targetBone {boneName=clavicle_r}}
     {targetBone {boneName=spine_02}}
     {targetBone {boneName=spine_03}}
    ]
    constraintWeights [1,1,1,1,1,1]
   }
   n7 = "Constraint - dyn_braid_bk_02_r (PPPPPPP)"!(
    ComponentPose=@n6
    Alpha=1
    ConstraintWeights_0=0.3392
    ConstraintWeights_1:float=0.2704
    ConstraintWeights_2:float=0.205
    ConstraintWeights_3:float=0.1136
    ConstraintWeights_4:float=0.0474
    ConstraintWeights_5:float=0.0204
    ConstraintWeights_6:float=0.0041
   ) -> (Pose) {
    alphaBoolBlend {blendOption=Linear}
    boneToModify {boneName=dyn_braid_bk_02_r}
    constraintSetup [
     {targetBone {boneName=neck_01}}
     {targetBone {boneName=spine_05}}
     {targetBone {boneName=neck_02}}
     {targetBone {boneName=head}}
     {targetBone {boneName=clavicle_r}}
     {targetBone {boneName=spine_04}}
     {targetBone {boneName=clavicle_l}}
    ]
    constraintWeights [1,1,1,1,1,1,1]
   }
   n8 = "Constraint - dyn_braid_bk_03_r (PPPPPP)"!(
    ComponentPose=@n7
    Alpha=1
    ConstraintWeights_0=0.5271
    ConstraintWeights_1:float=0.3274
    ConstraintWeights_2:float=0.0954
    ConstraintWeights_3:float=0.0387
    ConstraintWeights_4:float=0.0104
    ConstraintWeights_5:float=0.001
   ) -> (Pose) {
    alphaBoolBlend {blendOption=Linear}
    boneToModify {boneName=dyn_braid_bk_03_r}
    constraintSetup [
     {targetBone {boneName=spine_05}}
     {targetBone {boneName=clavicle_r}}
     {targetBone {boneName=clavicle_l}}
     {targetBone {boneName=spine_04}}
     {targetBone {boneName=spine_03}}
     {targetBone {boneName=spine_02}}
    ]
    constraintWeights [1,1,1,1,1,1]
   }
   n9 = "Constraint - dyn_braid_bk_04_r (PPPPPP)"!(
    ComponentPose=@n8
    Alpha=1
    ConstraintWeights_0=0.4229
    ConstraintWeights_1:float=0.2431
    ConstraintWeights_2:float=0.1295
    ConstraintWeights_3:float=0.1029
    ConstraintWeights_4:float=0.0642
    ConstraintWeights_5:float=0.0373
   ) -> (Pose) {
    alphaBoolBlend {blendOption=Linear}
    boneToModify {boneName=dyn_braid_bk_04_r}
    constraintSetup [
     {targetBone {boneName=spine_05}}
     {targetBone {boneName=spine_04}}
     {targetBone {boneName=clavicle_r}}
     {targetBone {boneName=clavicle_l}}
     {targetBone {boneName=spine_02}}
     {targetBone {boneName=spine_03}}
    ]
    constraintWeights [1,1,1,1,1,1]
   }
   n10 = "Constraint - dyn_braid_side_02_r (PPPPPPP)"!(
    ComponentPose=@n9
    Alpha=1
    ConstraintWeights_0=0.2817
    ConstraintWeights_1:float=0.2675
    ConstraintWeights_2:float=0.2036
    ConstraintWeights_3:float=0.1328
    ConstraintWeights_4:float=0.0968
    ConstraintWeights_5:float=0.0147
    ConstraintWeights_6:float=0.0029
   ) -> (Pose) {
    alphaBoolBlend {blendOption=Linear}
    boneToModify {boneName=dyn_braid_side_02_r}
    constraintSetup [
     {targetBone {boneName=neck_01}}
     {targetBone {boneName=spine_05}}
     {targetBone {boneName=neck_02}}
     {targetBone {boneName=head}}
     {targetBone {boneName=clavicle_r}}
     {targetBone {boneName=spine_04}}
     {targetBone {boneName=clavicle_l}}
    ]
    constraintWeights [1,1,1,1,1,1,1]
   }
   n11 = "Constraint - dyn_braid_side_03_r (PPPPP)"!(
    ComponentPose=@n10
    Alpha=1
    ConstraintWeights_0=0.4649
    ConstraintWeights_1:float=0.4203
    ConstraintWeights_2:float=0.0628
    ConstraintWeights_3:float=0.0438
    ConstraintWeights_4:float=0.0082
   ) -> (Pose) {
    alphaBoolBlend {blendOption=Linear}
    boneToModify {boneName=dyn_braid_side_03_r}
    constraintSetup [
     {targetBone {boneName=clavicle_r}}
     {targetBone {boneName=spine_05}}
     {targetBone {boneName=clavicle_l}}
     {targetBone {boneName=spine_04}}
     {targetBone {boneName=spine_03}}
    ]
    constraintWeights [1,1,1,1,1]
   }
   n12 = "Constraint - dyn_braid_side_04_r (PPPPPP)"!(
    ComponentPose=@n11
    Alpha=1
    ConstraintWeights_0=0.428
    ConstraintWeights_1:float=0.2517
    ConstraintWeights_2:float=0.1334
    ConstraintWeights_3:float=0.0789
    ConstraintWeights_4:float=0.0699
    ConstraintWeights_5:float=0.038
   ) -> (Pose) {
    alphaBoolBlend {blendOption=Linear}
    boneToModify {boneName=dyn_braid_side_04_r}
    constraintSetup [
     {targetBone {boneName=spine_05}}
     {targetBone {boneName=spine_04}}
     {targetBone {boneName=clavicle_r}}
     {targetBone {boneName=clavicle_l}}
     {targetBone {boneName=spine_02}}
     {targetBone {boneName=spine_03}}
    ]
    constraintWeights [1,1,1,1,1,1]
   }
   n13 = "Transform (Modify) Bone - Bone: dyn_braid_side_02_l"(ComponentPose=@n12, Alpha=1, Translation="0,-1,0", Rotation="0,15,0", Scale="1,1,1") {
    rotationMode=BMM_Additive
    rotationSpace=BCS_BoneSpace
    translationMode=BMM_Additive
    translationSpace=BCS_BoneSpace
    alphaBoolBlend {blendOption=Linear}
    boneToModify {boneName=dyn_braid_side_02_l}
    rotation {yaw=15}
    translation {y=-1}
   }
   n14 = "Transform (Modify) Bone - Bone: dyn_braid_bk_02_l"(ComponentPose=@n13, Alpha=1, Translation="0,-1,0", Rotation="0,15,0", Scale="1,1,1") {
    rotationMode=BMM_Additive
    rotationSpace=BCS_BoneSpace
    translationMode=BMM_Additive
    translationSpace=BCS_BoneSpace
    alphaBoolBlend {blendOption=Linear}
    boneToModify {boneName=dyn_braid_bk_02_l}
    rotation {yaw=15}
    translation {y=-1}
   }
   n15 = "Transform (Modify) Bone - Bone: dyn_braid_bk_02_r"(ComponentPose=@n14, Alpha=1, Translation="0,-1,0", Rotation="0,15,0", Scale="1,1,1") {
    rotationMode=BMM_Additive
    rotationSpace=BCS_BoneSpace
    translationMode=BMM_Additive
    translationSpace=BCS_BoneSpace
    alphaBoolBlend {blendOption=Linear}
    boneToModify {boneName=dyn_braid_bk_02_r}
    rotation {yaw=15}
    translation {y=-1}
   }
   n16 = "Transform (Modify) Bone - Bone: dyn_braid_side_02_r"(ComponentPose=@n15, Alpha=1, Translation="0,-1,0", Rotation="0,15,0", Scale="1,1,1") {
    rotationMode=BMM_Additive
    rotationSpace=BCS_BoneSpace
    translationMode=BMM_Additive
    translationSpace=BCS_BoneSpace
    alphaBoolBlend {blendOption=Linear}
    boneToModify {boneName=dyn_braid_side_02_r}
    rotation {yaw=15}
    translation {y=-1}
   }
   n17 = "Transform (Modify) Bone - Bone: dyn_braid_side_03_l"(ComponentPose=@n16, Alpha=1, Translation="0,-2.5,0", Rotation="0,0,0", Scale="1,1,1") {
    rotationMode=BMM_Additive
    rotationSpace=BCS_BoneSpace
    translationMode=BMM_Additive
    translationSpace=BCS_BoneSpace
    alphaBoolBlend {blendOption=Linear}
    boneToModify {boneName=dyn_braid_side_03_l}
    translation {y=-2.5}
   }
   n18 = "Transform (Modify) Bone - Bone: dyn_braid_bk_03_l"(ComponentPose=@n17, Alpha=1, Translation="0,-2,0", Rotation="0,0,0", Scale="1,1,1") {
    rotationMode=BMM_Additive
    rotationSpace=BCS_BoneSpace
    translationMode=BMM_Additive
    translationSpace=BCS_BoneSpace
    alphaBoolBlend {blendOption=Linear}
    boneToModify {boneName=dyn_braid_bk_03_l}
    translation {y=-2}
   }
   n19 = "Transform (Modify) Bone - Bone: dyn_braid_bk_03_r"(ComponentPose=@n18, Alpha=1, Translation="0,-2,0", Rotation="0,0,0", Scale="1,1,1") {
    rotationMode=BMM_Additive
    rotationSpace=BCS_BoneSpace
    translationMode=BMM_Additive
    translationSpace=BCS_BoneSpace
    alphaBoolBlend {blendOption=Linear}
    boneToModify {boneName=dyn_braid_bk_03_r}
    translation {y=-2}
   }
   "Transform (Modify) Bone - Bone: dyn_braid_side_03_r"(ComponentPose=@n19, Alpha=1, Translation="0,-3,0", Rotation="0,0,0", Scale="1,1,1") {
    rotationMode=BMM_Additive
    rotationSpace=BCS_BoneSpace
    translationMode=BMM_Additive
    translationSpace=BCS_BoneSpace
    alphaBoolBlend {blendOption=Linear}
    boneToModify {boneName=dyn_braid_side_03_r}
    translation {y=-3}
   }
   n20 = Look At - Bone: dyn_braid_side_01_l(ComponentPose=Local To Component(LocalPose=Use cached pose 'PreDyn'()), Alpha=1) {
    alphaBoolBlend {blendOption=Linear}
    boneToModify {boneName=dyn_braid_side_01_l}
    lookAtLocation=<5 -10 -9>
    lookAtTarget {boneReference {boneName=spine_05}}
    lookAt_Axis {axis=<1 0>}
    lookUp_Axis {axis=<0 1>}
   }
   n21 = Look At - Bone: dyn_braid_bk_01_l(ComponentPose=@n20, Alpha=1) {
    alphaBoolBlend {blendOption=Linear}
    boneToModify {boneName=dyn_braid_bk_01_l}
    lookAtLocation=<5 -10 -5>
    lookAtTarget {boneReference {boneName=spine_05}}
    lookAt_Axis {axis=<1 0>}
    lookUp_Axis {axis=<0 1>}
   }
   n22 = Look At - Bone: dyn_braid_bk_01_r(ComponentPose=@n21, Alpha=1) {
    alphaBoolBlend {blendOption=Linear}
    boneToModify {boneName=dyn_braid_bk_01_r}
    lookAtLocation=<5 -10 5>
    lookAtTarget {boneReference {boneName=spine_05}}
    lookAt_Axis {axis=<1 0>}
    lookUp_Axis {axis=<0 1>}
   }
   n23 = Look At - Bone: dyn_braid_side_01_r(ComponentPose=@n22, Alpha=1) {
    alphaBoolBlend {blendOption=Linear}
    boneToModify {boneName=dyn_braid_side_01_r}
    lookAtLocation=<5 -10 5>
    lookAtTarget {boneReference {boneName=spine_05}}
    lookAt_Axis {axis=<1 0>}
    lookUp_Axis {axis=<0 1>}
   }
   n24 = RigidBody!(
    ComponentPose=@n23
    bAlphaBoolEnabled:bool
    OverrideWorldGravity:Vector="0,0,0"
    ComponentLinearAccScale:Vector="0,0,0"
    ComponentLinearVelScale:Vector="0,0,0"
    bTransferBoneVelocities:bool=False
   ) -> (Pose) {
    alphaInputType=Bool
    bClampLinearTranslationLimitToRefPose=true
    bFreezeIncomingPoseOnStart=true
    bOverrideWorldGravity=true
    bTransferBoneVelocities=true
    overridePhysicsAsset=&/F_MED_ScrapTunnel_FaceAcc_Physics
    simulationSpace=BaseBoneSpace
    alphaBoolBlend {blendOption=Linear}
    baseBoneRef {boneName=spine_05}
    componentAppliedLinearAccClamp=<100 100 100>
    componentLinearAccScale=<0.6 0.2 0.2>
    componentLinearVelScale=<0.6 0.2 0.2>
    simSpaceSettings {maxAngularAcceleration=40, maxAngularVelocity=8, maxLinearAcceleration=100, maxLinearVelocity=700, worldAlpha=0.2}
   }
   n25 = "Transform (Modify) Bone - Bone: dyn_braid_side_base_r"(ComponentPose=Local To Component(LocalPose=Use cached pose 'PreDyn'()), Alpha=1, Translation="0,0,0", Rotation="0,0,0", Scale="1,1,1") {rotationMode=BMM_Additive, rotationSpace=BCS_BoneSpace, alphaBoolBlend {blendOption=Linear}, boneToModify {boneName=dyn_braid_side_base_r}}
   n26 = "Transform (Modify) Bone - Bone: dyn_braid_bk_base_r"(ComponentPose=@n25, Alpha=1, Translation="0,0,0", Rotation="0,0,0", Scale="1,1,1") {rotationMode=BMM_Additive, rotationSpace=BCS_BoneSpace, alphaBoolBlend {blendOption=Linear}, boneToModify {boneName=dyn_braid_bk_base_r}}
   n27 = "Transform (Modify) Bone - Bone: dyn_braid_bk_base_r"(ComponentPose=@n26, Alpha=1, Translation="0,0,0", Rotation="0,0,0", Scale="1,1,1") {rotationMode=BMM_Additive, rotationSpace=BCS_BoneSpace, alphaBoolBlend {blendOption=Linear}, boneToModify {boneName=dyn_braid_bk_base_r}}
   n28 = "Transform (Modify) Bone - Bone: dyn_braid_bk_base_r"(ComponentPose=@n27, Alpha=1, Translation="0,0,0", Rotation="0,0,0", Scale="1,1,1") {rotationMode=BMM_Additive, rotationSpace=BCS_BoneSpace, alphaBoolBlend {blendOption=Linear}, boneToModify {boneName=dyn_braid_bk_base_r}}
   n29 = RigidBody(
    ComponentPose=@n28
    OverrideWorldGravity="0,0,0"
    ComponentLinearAccScale="0,0,0"
    ComponentLinearVelScale="0,0,0"
    bTransferBoneVelocities=False
   ) {
    alphaInputType=Bool
    bClampLinearTranslationLimitToRefPose=true
    bFreezeIncomingPoseOnStart=true
    bOverrideWorldGravity=true
    bTransferBoneVelocities=true
    overridePhysicsAsset=&/F_MED_ScrapTunnel_FaceAcc_Physics_Skydive
    simulationSpace=BaseBoneSpace
    alphaBoolBlend {blendOption=Linear}
    baseBoneRef {boneName=neck_01}
    componentAppliedLinearAccClamp=<100 100 100>
    componentLinearAccScale=<0.6 0.2 0.2>
    componentLinearVelScale=<0.6 0.2 0.2>
    simSpaceSettings {maxAngularAcceleration=40, maxLinearAcceleration=100, maxLinearVelocity=700, worldAlpha=1}
   }
   n30 = Blend Poses by bool(BlendPose_0=Component To Local(ComponentPose=@n29), BlendPose_1=Component To Local(ComponentPose=@n24)) {blendType=Linear}
   n31 = Look At - Bone: dyn_braid_side_01_l(ComponentPose=Local To Component(LocalPose=Use cached pose 'PreDyn'()), Alpha=1) {
    alphaBoolBlend {blendOption=Linear}
    boneToModify {boneName=dyn_braid_side_01_l}
    lookAtLocation=<5 -10 -5>
    lookAtTarget {boneReference {boneName=spine_05}}
    lookAt_Axis {axis=<1 0>}
    lookUp_Axis {axis=<0 1>}
   }
   n32 = Look At - Bone: dyn_braid_bk_01_l(ComponentPose=@n31, Alpha=1) {
    alphaBoolBlend {blendOption=Linear}
    boneToModify {boneName=dyn_braid_bk_01_l}
    lookAtLocation=<5 -10 -2>
    lookAtTarget {boneReference {boneName=spine_05}}
    lookAt_Axis {axis=<1 0>}
    lookUp_Axis {axis=<0 1>}
   }
   n33 = Look At - Bone: dyn_braid_bk_01_r(ComponentPose=@n32, Alpha=1) {
    alphaBoolBlend {blendOption=Linear}
    boneToModify {boneName=dyn_braid_bk_01_r}
    lookAtLocation=<5 -10 2>
    lookAtTarget {boneReference {boneName=spine_05}}
    lookAt_Axis {axis=<1 0>}
    lookUp_Axis {axis=<0 1>}
   }
   n34 = Look At - Bone: dyn_braid_side_01_r(ComponentPose=@n33, Alpha=1) {
    alphaBoolBlend {blendOption=Linear}
    boneToModify {boneName=dyn_braid_side_01_r}
    lookAtLocation=<5 -10 5>
    lookAtTarget {boneReference {boneName=spine_05}}
    lookAt_Axis {axis=<1 0>}
    lookUp_Axis {axis=<0 1>}
   }
   n35 = RigidBody(
    ComponentPose=@n34
    OverrideWorldGravity="0,0,0"
    ComponentLinearAccScale="0,0,0"
    ComponentLinearVelScale="0,0,0"
    bTransferBoneVelocities=true
   ) {
    alphaInputType=Bool
    bClampLinearTranslationLimitToRefPose=true
    bFreezeIncomingPoseOnStart=true
    bOverrideWorldGravity=true
    bTransferBoneVelocities=true
    overridePhysicsAsset=&/F_MED_ScrapTunnel_FaceAcc_Physics_Backpack
    simulationSpace=BaseBoneSpace
    alphaBoolBlend {blendOption=Linear}
    baseBoneRef {boneName=head}
    componentAppliedLinearAccClamp=<100 100 100>
    componentLinearAccScale=<0.6 0.2 0.2>
    componentLinearVelScale=<0.6 0.2 0.2>
    simSpaceSettings {maxAngularAcceleration=40, maxAngularVelocity=8, maxLinearAcceleration=100, maxLinearVelocity=700, worldAlpha=0.2}
   }
   n36 = Component To Local(ComponentPose=@n35)
   n37 = Look At - Bone: dyn_braid_side_01_l(ComponentPose=Local To Component(LocalPose=Use cached pose 'PreDyn'()), Alpha=1) {
    alphaBoolBlend {blendOption=Linear}
    boneToModify {boneName=dyn_braid_side_01_l}
    lookAtLocation=<5 -10 -5>
    lookAtTarget {boneReference {boneName=spine_05}}
    lookAt_Axis {axis=<1 0>}
    lookUp_Axis {axis=<0 1>}
   }
   n38 = Look At - Bone: dyn_braid_bk_01_l(ComponentPose=@n37, Alpha=1) {
    alphaBoolBlend {blendOption=Linear}
    boneToModify {boneName=dyn_braid_bk_01_l}
    lookAtLocation=<5 -10 -2>
    lookAtTarget {boneReference {boneName=spine_05}}
    lookAt_Axis {axis=<1 0>}
    lookUp_Axis {axis=<0 1>}
   }
   n39 = Look At - Bone: dyn_braid_bk_01_r(ComponentPose=@n38, Alpha=1) {
    alphaBoolBlend {blendOption=Linear}
    boneToModify {boneName=dyn_braid_bk_01_r}
    lookAtLocation=<5 -10 2>
    lookAtTarget {boneReference {boneName=spine_05}}
    lookAt_Axis {axis=<1 0>}
    lookUp_Axis {axis=<0 1>}
   }
   n40 = Look At - Bone: dyn_braid_side_01_r(ComponentPose=@n39, Alpha=1) {
    alphaBoolBlend {blendOption=Linear}
    boneToModify {boneName=dyn_braid_side_01_r}
    lookAtLocation=<5 -10 5>
    lookAtTarget {boneReference {boneName=spine_05}}
    lookAt_Axis {axis=<1 0>}
    lookUp_Axis {axis=<0 1>}
   }
   n41 = RigidBody(
    ComponentPose=@n40
    OverrideWorldGravity="0,0,0"
    ComponentLinearAccScale="0,0,0"
    ComponentLinearVelScale="0,0,0"
    bTransferBoneVelocities=true
   ) {
    alphaInputType=Bool
    bClampLinearTranslationLimitToRefPose=true
    bFreezeIncomingPoseOnStart=true
    bOverrideWorldGravity=true
    bTransferBoneVelocities=true
    overridePhysicsAsset=&/F_MED_ScrapTunnel_FaceAcc_Physics_Cape
    simulationSpace=BaseBoneSpace
    alphaBoolBlend {blendOption=Linear}
    baseBoneRef {boneName=head}
    componentAppliedLinearAccClamp=<100 100 100>
    componentLinearAccScale=<0.6 0.2 0.2>
    componentLinearVelScale=<0.6 0.2 0.2>
    simSpaceSettings {maxAngularAcceleration=40, maxAngularVelocity=8, maxLinearAcceleration=100, maxLinearVelocity=700, worldAlpha=0.2}
   }
   n42 = Component To Local(ComponentPose=@n41)
   n43 = "Constraint - dyn_braid_side_02_l (PPPPPPP)"(
    ComponentPose=Local To Component(LocalPose=Blend Poses by bool(BlendPose_0=@n42, BlendPose_1=@n36) {blendType=Linear})
    Alpha=0.5
    ConstraintWeights_0=0.3194
    ConstraintWeights_1=0.2484
    ConstraintWeights_2=0.2229
    ConstraintWeights_3=0.1248
    ConstraintWeights_4=0.0651
    ConstraintWeights_5=0.0183
    ConstraintWeights_6=0.0011
   ) {
    alphaBoolBlend {blendOption=Linear}
    boneToModify {boneName=dyn_braid_side_02_l}
    constraintSetup [
     {targetBone {boneName=neck_01}}
     {targetBone {boneName=spine_05}}
     {targetBone {boneName=neck_02}}
     {targetBone {boneName=head}}
     {targetBone {boneName=clavicle_l}}
     {targetBone {boneName=spine_04}}
     {targetBone {boneName=clavicle_r}}
    ]
    constraintWeights [1,1,1,1,1,1,1]
   }
   n44 = "Constraint - dyn_braid_side_03_l (PPPPP)"(
    ComponentPose=@n43
    Alpha=1
    ConstraintWeights_0=0.4756
    ConstraintWeights_1=0.4188
    ConstraintWeights_2=0.0578
    ConstraintWeights_3=0.0396
    ConstraintWeights_4=0.0082
   ) {
    alphaBoolBlend {blendOption=Linear}
    boneToModify {boneName=dyn_braid_side_03_l}
    constraintSetup [
     {targetBone {boneName=clavicle_l}}
     {targetBone {boneName=spine_05}}
     {targetBone {boneName=clavicle_r}}
     {targetBone {boneName=spine_04}}
     {targetBone {boneName=spine_03}}
    ]
    constraintWeights [1,1,1,1,1]
   }
   n45 = "Constraint - dyn_braid_side_04_l (PPPPPP)"(
    ComponentPose=@n44
    Alpha=1
    ConstraintWeights_0=0.428
    ConstraintWeights_1=0.2517
    ConstraintWeights_2=0.1334
    ConstraintWeights_3=0.0789
    ConstraintWeights_4=0.0699
    ConstraintWeights_5=0.038
   ) {
    alphaBoolBlend {blendOption=Linear}
    boneToModify {boneName=dyn_braid_side_04_l}
    constraintSetup [
     {targetBone {boneName=spine_05}}
     {targetBone {boneName=spine_04}}
     {targetBone {boneName=clavicle_l}}
     {targetBone {boneName=clavicle_r}}
     {targetBone {boneName=spine_02}}
     {targetBone {boneName=spine_03}}
    ]
    constraintWeights [1,1,1,1,1,1]
   }
   n46 = "Constraint - dyn_braid_bk_02_l (PPPPPPP)"(
    ComponentPose=@n45
    Alpha=0.5
    ConstraintWeights_0=0.3803
    ConstraintWeights_1=0.2329
    ConstraintWeights_2=0.2181
    ConstraintWeights_3=0.1151
    ConstraintWeights_4=0.0284
    ConstraintWeights_5=0.0225
    ConstraintWeights_6=0.0028
   ) {
    alphaBoolBlend {blendOption=Linear}
    boneToModify {boneName=dyn_braid_bk_02_l}
    constraintSetup [
     {targetBone {boneName=neck_01}}
     {targetBone {boneName=neck_02}}
     {targetBone {boneName=spine_05}}
     {targetBone {boneName=head}}
     {targetBone {boneName=clavicle_l}}
     {targetBone {boneName=spine_04}}
     {targetBone {boneName=clavicle_r}}
    ]
    constraintWeights [1,1,1,1,1,1,1]
   }
   n47 = "Constraint - dyn_braid_bk_03_l (PPPPPP)"(
    ComponentPose=@n46
    Alpha=1
    ConstraintWeights_0=0.5271
    ConstraintWeights_1=0.3273
    ConstraintWeights_2=0.0954
    ConstraintWeights_3=0.0387
    ConstraintWeights_4=0.0104
    ConstraintWeights_5=0.001
   ) {
    alphaBoolBlend {blendOption=Linear}
    boneToModify {boneName=dyn_braid_bk_03_l}
    constraintSetup [
     {targetBone {boneName=spine_05}}
     {targetBone {boneName=clavicle_l}}
     {targetBone {boneName=clavicle_r}}
     {targetBone {boneName=spine_04}}
     {targetBone {boneName=spine_03}}
     {targetBone {boneName=spine_02}}
    ]
    constraintWeights [1,1,1,1,1,1]
   }
   n48 = "Constraint - dyn_braid_bk_04_l (PPPPPP)"(
    ComponentPose=@n47
    Alpha=1
    ConstraintWeights_0=0.4229
    ConstraintWeights_1=0.2431
    ConstraintWeights_2=0.1295
    ConstraintWeights_3=0.1029
    ConstraintWeights_4=0.0642
    ConstraintWeights_5=0.0373
   ) {
    alphaBoolBlend {blendOption=Linear}
    boneToModify {boneName=dyn_braid_bk_04_l}
    constraintSetup [
     {targetBone {boneName=spine_05}}
     {targetBone {boneName=spine_04}}
     {targetBone {boneName=clavicle_l}}
     {targetBone {boneName=clavicle_r}}
     {targetBone {boneName=spine_02}}
     {targetBone {boneName=spine_03}}
    ]
    constraintWeights [1,1,1,1,1,1]
   }
   n49 = "Constraint - dyn_braid_bk_02_r (PPPPPPP)"(
    ComponentPose=@n48
    Alpha=0.5
    ConstraintWeights_0=0.3392
    ConstraintWeights_1=0.2704
    ConstraintWeights_2=0.205
    ConstraintWeights_3=0.1136
    ConstraintWeights_4=0.0474
    ConstraintWeights_5=0.0204
    ConstraintWeights_6=0.0041
   ) {
    alphaBoolBlend {blendOption=Linear}
    boneToModify {boneName=dyn_braid_bk_02_r}
    constraintSetup [
     {targetBone {boneName=neck_01}}
     {targetBone {boneName=spine_05}}
     {targetBone {boneName=neck_02}}
     {targetBone {boneName=head}}
     {targetBone {boneName=clavicle_r}}
     {targetBone {boneName=spine_04}}
     {targetBone {boneName=clavicle_l}}
    ]
    constraintWeights [1,1,1,1,1,1,1]
   }
   n50 = "Constraint - dyn_braid_bk_03_r (PPPPPP)"(
    ComponentPose=@n49
    Alpha=1
    ConstraintWeights_0=0.5271
    ConstraintWeights_1=0.3274
    ConstraintWeights_2=0.0954
    ConstraintWeights_3=0.0387
    ConstraintWeights_4=0.0104
    ConstraintWeights_5=0.001
   ) {
    alphaBoolBlend {blendOption=Linear}
    boneToModify {boneName=dyn_braid_bk_03_r}
    constraintSetup [
     {targetBone {boneName=spine_05}}
     {targetBone {boneName=clavicle_r}}
     {targetBone {boneName=clavicle_l}}
     {targetBone {boneName=spine_04}}
     {targetBone {boneName=spine_03}}
     {targetBone {boneName=spine_02}}
    ]
    constraintWeights [1,1,1,1,1,1]
   }
   n51 = "Constraint - dyn_braid_bk_04_r (PPPPPP)"(
    ComponentPose=@n50
    Alpha=1
    ConstraintWeights_0=0.4229
    ConstraintWeights_1=0.2431
    ConstraintWeights_2=0.1295
    ConstraintWeights_3=0.1029
    ConstraintWeights_4=0.0642
    ConstraintWeights_5=0.0373
   ) {
    alphaBoolBlend {blendOption=Linear}
    boneToModify {boneName=dyn_braid_bk_04_r}
    constraintSetup [
     {targetBone {boneName=spine_05}}
     {targetBone {boneName=spine_04}}
     {targetBone {boneName=clavicle_r}}
     {targetBone {boneName=clavicle_l}}
     {targetBone {boneName=spine_02}}
     {targetBone {boneName=spine_03}}
    ]
    constraintWeights [1,1,1,1,1,1]
   }
   n52 = "Constraint - dyn_braid_side_02_r (PPPPPPP)"(
    ComponentPose=@n51
    Alpha=0.5
    ConstraintWeights_0=0.2817
    ConstraintWeights_1=0.2675
    ConstraintWeights_2=0.2036
    ConstraintWeights_3=0.1328
    ConstraintWeights_4=0.0968
    ConstraintWeights_5=0.0147
    ConstraintWeights_6=0.0029
   ) {
    alphaBoolBlend {blendOption=Linear}
    boneToModify {boneName=dyn_braid_side_02_r}
    constraintSetup [
     {targetBone {boneName=neck_01}}
     {targetBone {boneName=spine_05}}
     {targetBone {boneName=neck_02}}
     {targetBone {boneName=head}}
     {targetBone {boneName=clavicle_r}}
     {targetBone {boneName=spine_04}}
     {targetBone {boneName=clavicle_l}}
    ]
    constraintWeights [1,1,1,1,1,1,1]
   }
   n53 = "Constraint - dyn_braid_side_03_r (PPPPP)"(
    ComponentPose=@n52
    Alpha=1
    ConstraintWeights_0=0.4649
    ConstraintWeights_1=0.4203
    ConstraintWeights_2=0.0628
    ConstraintWeights_3=0.0438
    ConstraintWeights_4=0.0082
   ) {
    alphaBoolBlend {blendOption=Linear}
    boneToModify {boneName=dyn_braid_side_03_r}
    constraintSetup [
     {targetBone {boneName=clavicle_r}}
     {targetBone {boneName=spine_05}}
     {targetBone {boneName=clavicle_l}}
     {targetBone {boneName=spine_04}}
     {targetBone {boneName=spine_03}}
    ]
    constraintWeights [1,1,1,1,1]
   }
   n54 = "Constraint - dyn_braid_side_04_r (PPPPPP)"(
    ComponentPose=@n53
    Alpha=1
    ConstraintWeights_0=0.428
    ConstraintWeights_1=0.2517
    ConstraintWeights_2=0.1334
    ConstraintWeights_3=0.0789
    ConstraintWeights_4=0.0699
    ConstraintWeights_5=0.038
   ) {
    alphaBoolBlend {blendOption=Linear}
    boneToModify {boneName=dyn_braid_side_04_r}
    constraintSetup [
     {targetBone {boneName=spine_05}}
     {targetBone {boneName=spine_04}}
     {targetBone {boneName=clavicle_r}}
     {targetBone {boneName=clavicle_l}}
     {targetBone {boneName=spine_02}}
     {targetBone {boneName=spine_03}}
    ]
    constraintWeights [1,1,1,1,1,1]
   }
   n55 = "Transform (Modify) Bone - Bone: dyn_braid_side_03_l"(ComponentPose=@n54, Alpha=1, Translation="0,-1,0", Rotation="0,15,0", Scale="1,1,1") {
    rotationMode=BMM_Additive
    rotationSpace=BCS_BoneSpace
    translationMode=BMM_Additive
    translationSpace=BCS_BoneSpace
    alphaBoolBlend {blendOption=Linear}
    boneToModify {boneName=dyn_braid_side_03_l}
    rotation {yaw=15}
    translation {y=-1}
   }
   n56 = "Transform (Modify) Bone - Bone: dyn_braid_bk_03_l"(ComponentPose=@n55, Alpha=1, Translation="0,-1,0", Rotation="0,15,0", Scale="1,1,1") {
    rotationMode=BMM_Additive
    rotationSpace=BCS_BoneSpace
    translationMode=BMM_Additive
    translationSpace=BCS_BoneSpace
    alphaBoolBlend {blendOption=Linear}
    boneToModify {boneName=dyn_braid_bk_03_l}
    rotation {yaw=15}
    translation {y=-1}
   }
   n57 = "Transform (Modify) Bone - Bone: dyn_braid_bk_03_r"(ComponentPose=@n56, Alpha=1, Translation="0,-1,0", Rotation="0,15,0", Scale="1,1,1") {
    rotationMode=BMM_Additive
    rotationSpace=BCS_BoneSpace
    translationMode=BMM_Additive
    translationSpace=BCS_BoneSpace
    alphaBoolBlend {blendOption=Linear}
    boneToModify {boneName=dyn_braid_bk_03_r}
    rotation {yaw=15}
    translation {y=-1}
   }
   n58 = "Transform (Modify) Bone - Bone: dyn_braid_side_03_r"(ComponentPose=@n57, Alpha=1, Translation="0,-1,0", Rotation="0,15,0", Scale="1,1,1") {
    rotationMode=BMM_Additive
    rotationSpace=BCS_BoneSpace
    translationMode=BMM_Additive
    translationSpace=BCS_BoneSpace
    alphaBoolBlend {blendOption=Linear}
    boneToModify {boneName=dyn_braid_side_03_r}
    rotation {yaw=15}
    translation {y=-1}
   }
   n59 = Component To Local(ComponentPose=@n58)
   n60 = Look At - Bone: dyn_braid_side_01_l(ComponentPose=Local To Component(LocalPose=Use cached pose 'PreDyn'()), Alpha=1) {
    alphaBoolBlend {blendOption=Linear}
    boneToModify {boneName=dyn_braid_side_01_l}
    lookAtLocation=<5 -10 -5>
    lookAtTarget {boneReference {boneName=spine_05}}
    lookAt_Axis {axis=<1 0>}
    lookUp_Axis {axis=<0 1>}
   }
   n61 = Look At - Bone: dyn_braid_bk_01_l(ComponentPose=@n60, Alpha=1) {
    alphaBoolBlend {blendOption=Linear}
    boneToModify {boneName=dyn_braid_bk_01_l}
    lookAtLocation=<5 -10 -2>
    lookAtTarget {boneReference {boneName=spine_05}}
    lookAt_Axis {axis=<1 0>}
    lookUp_Axis {axis=<0 1>}
   }
   n62 = Look At - Bone: dyn_braid_bk_01_r(ComponentPose=@n61, Alpha=1) {
    alphaBoolBlend {blendOption=Linear}
    boneToModify {boneName=dyn_braid_bk_01_r}
    lookAtLocation=<5 -10 2>
    lookAtTarget {boneReference {boneName=spine_05}}
    lookAt_Axis {axis=<1 0>}
    lookUp_Axis {axis=<0 1>}
   }
   n63 = Look At - Bone: dyn_braid_side_01_r(ComponentPose=@n62, Alpha=1) {
    alphaBoolBlend {blendOption=Linear}
    boneToModify {boneName=dyn_braid_side_01_r}
    lookAtLocation=<5 -10 5>
    lookAtTarget {boneReference {boneName=spine_05}}
    lookAt_Axis {axis=<1 0>}
    lookUp_Axis {axis=<0 1>}
   }
   n64 = RigidBody!(
    ComponentPose=@n63
    Alpha:float
    OverrideWorldGravity="0,0,0"
    ComponentLinearAccScale="0,0,0"
    ComponentLinearVelScale="0,0,0"
    bTransferBoneVelocities=False
   ) -> (Pose) {
    bClampLinearTranslationLimitToRefPose=true
    bFreezeIncomingPoseOnStart=true
    bOverrideWorldGravity=true
    bTransferBoneVelocities=true
    overridePhysicsAsset=&/F_MED_ScrapTunnel_FaceAcc_Physics_FE
    simulationSpace=BaseBoneSpace
    alphaBoolBlend {blendOption=Linear}
    baseBoneRef {boneName=spine_05}
    componentAppliedLinearAccClamp=<100 100 100>
    componentLinearAccScale=<0.6 0.2 0.2>
    componentLinearVelScale=<0.6 0.2 0.2>
    simSpaceSettings {worldAlpha=1}
   }
   n65 = Control Rig!(
    Source=Blend Poses by bool(BlendPose_0=Component To Local(ComponentPose=@n64), BlendPose_1=Blend Poses by bool(BlendPose_0=@n59, BlendPose_1=@n30) {blendType=Linear}) {blendType=Linear}
    Alpha
    FE:bool=false
   ) -> (Pose) {
    alphaBoolBlend {blendOption=Linear}
    controlRigAssetReference {blueprintRigClass=&/F_MED_ScrapTunnel_FaceAcc_CtrlRig_PostDyn.F_MED_ScrapTunnel_FaceAcc_CtrlRig_PostDyn_C}
   }
   Output Pose(Result=@n65)
   n66 = Layered blend per bone(
    BasePose=Copy Pose From Mesh() {bUseAttachedParent=true}
    BlendPoses_0=Copy Pose From Mesh(SourceMeshComponent=Get Body_Mesh()) {bUseAttachedParent=true}
   ) {layerSetup [{branchFilters [{boneName=clavicle_l}, {boneName=clavicle_r}, {boneName=upperarm_r}, {boneName=upperarm_l}]}]}
   n67 = Control Rig(Source=@n66, FE=false) {
    alphaBoolBlend {blendOption=Linear}
    controlRigAssetReference {blueprintRigClass=&/F_MED_ScrapTunnel_FaceAcc_CtrlRig_PreDyn.F_MED_ScrapTunnel_FaceAcc_CtrlRig_PreDyn_C}
   }
   Save cached pose 'PreDyn'(Pose=@n67) {cacheName=PreDyn}
  ]}
 EventGraph {nodes [
   Event BlueprintInitializeAnimation()
   > Set Skel Mesh(Skel Mesh=GetOwnerSkeletalMeshComponent())
   Event BlueprintUpdateAnimation()
   > n1 = Sequence!(execute) -> (then_0, then_1, then_2:Exec, then_3:Exec)
   n2 = Sequence!(execute=@n1.then_1) -> (then_0, then_1, then_2, then_3, then_4:Exec, then_5:Exec)
   n3 = Get isWearingPack() -> (isWearingPack:bool)
   Set skydive(execute=@n2.then_4, skydive:bool=AND Boolean(A=Get bIsSkyDiving(), B=NOT Boolean(A=@n3))) -> (Output_Get:bool)
   Set isWearingPack(
    execute=@n2.then_1
    isWearingPack:bool=OR Boolean(A=Get isWearingBackpack() -> (isWearingBackpack:bool), B=Get isWearingCape() -> (isWearingCape:bool))
   ) -> (Output_Get:bool)
   n4 = Get bIsPlayingEmote()
   Set FE Idel(execute=@n2.then_0, FE Idel:bool=AND Boolean(A=Get bIsInFrontEnd(), B=NOT Boolean(A=@n4))) -> (Output_Get:bool)
   n5 = Get isWearingPack()
   n6 = Get bIsSkyDiving()
   Set DynHair(execute=@n2.then_5, DynHair:bool=AND Boolean(A=NOT Boolean(A=@n6), B=NOT Boolean(A=@n5))) -> (Output_Get:bool)
   Set DynHairCape(execute=@n2.then_3, DynHairCape:bool=Get isWearingCape()) -> (Output_Get:bool)
   Set DynHairBackpack(execute=@n2.then_2, DynHairBackpack:bool=Get isWearingBackpack()) -> (Output_Get:bool)
   n7 = Sequence!(execute=@n1.then_2) -> (then_0, then_1)
   n8 = Sequence!(execute=@n7.then_1) -> (then_0, then_1, then_2:Exec, then_3:Exec)
   n9 = float * float(A=GetGameTimeInSeconds(), B=1.125) {is=title K2Node_PromotableOperator}
   > n10 = RandomRotatorFunction!(
    execute
    GameTimeInSeconds:double=@n9
    PerlinScale:double=1
    Frequency:Rotator="5,2.8,5"
    Amplitude:Rotator="10,5,3"
    Offset:Rotator="-10,-3,0"
    TimeOffset:double=0
   ) -> (then, RandomRotator:Rotator)
   > Set TesselRotator_01(TesselRotator_01:Rotator=@n10) -> (Output_Get:Rotator)
   n11 = RandomRotatorFunction(execute=@n8.then_1, GameTimeInSeconds=@n9, PerlinScale=1, Frequency="4.5,1.8,4.5", Amplitude="10,5,3", Offset="-10,3,1", TimeOffset=0)
   > Set TesselRotator_02(TesselRotator_02:Rotator=@n11) -> (Output_Get:Rotator)
   n12 = RandomRotatorFunction(execute=@n8.then_2, GameTimeInSeconds=@n9, PerlinScale=1, Frequency="3,5,3", Amplitude="10,5,3", Offset="-10,3,5", TimeOffset=0)
   > Set TesselRotator_03(TesselRotator_03:Rotator=@n12) -> (Output_Get:Rotator)
   n13 = RandomRotatorFunction(execute=@n8.then_3, GameTimeInSeconds=@n9, PerlinScale=1, Frequency="4,2,3.5", Amplitude="10,5,3", Offset="-10,3,3", TimeOffset=0)
   > Set TesselRotator_04(TesselRotator_04:Rotator=@n13) -> (Output_Get:Rotator)
   n14 = Get isWearingPack()
   n15 = Branch(execute=@n7.then_0, Condition=AND Boolean(A=Get bIsSkyDiving(), B=NOT Boolean(A=@n14)))
   > Set Tassle 01 Alpha(Tassle 01 Alpha:double=0.25) -> (Output_Get:double)
   > Set Tassle 02 Alpha(Tassle 02 Alpha:double=0.15) -> (Output_Get:double)
   > Set Tassle 03 Alpha(Tassle 03 Alpha:double=0.3) -> (Output_Get:double)
   > Set Tassle 04 Alpha(Tassle 04 Alpha:double=0.2) -> (Output_Get:double)
   Set Tassle 01 Alpha(execute=@n15.else, Tassle 01 Alpha=0)
   > Set Tassle 02 Alpha(Tassle 02 Alpha=0)
   > Set Tassle 03 Alpha(Tassle 03 Alpha=0)
   > Set Tassle 04 Alpha(Tassle 04 Alpha=0)
   Set head_up(execute=@n1.then_3, head_up:float=GetCurveValue(CurveName=head_up)) -> (Output_Get:float)
   n16 = Gravity Override State Function DEPRECATED(
    execute=@n1.then_0
    CharacterPartAnimInstance=Self-Reference()
    GameTimeInSeconds=GetGameTimeInSeconds()
    GravityOverrideParams=Get Gravity Override Params()
    RBAN Params=Get RBAN Params()
    DebugJointName=spine_05
   )
   n17 = Get bIsSkyDiving()
   n18 = Get bIsJumping()
   n19 = Get bIsFalling()
   > n20 = Branch(Condition=AND Boolean(A=OR Boolean(A=@n18, B=@n19), B=NOT Boolean(A=@n17)))
   > Set GravityOverride(GravityOverride="0,0,300")
   Set GravityOverride(execute=@n20.else, GravityOverride=@n16)
   Event OnCharacterCustomizationLoaded()
   > n21 = Sequence!(execute) -> (then_0, then_1)
   n22 = CheckForBackbling()
   > Set isWearingBackpack(isWearingBackpack:bool=AND Boolean(A=@n22, B=NOT Boolean(A=@n22.bIsBackblingLong))) -> (Output_Get:bool)
   > Set isWearingCape(isWearingCape=AND Boolean(A=@n22, B=@n22.bIsBackblingLong))
   n23 = Cast To PlayerPawn_Athena(execute=@n21.then_1, Object=TryGetPawnOwner())
   > Set Body_Mesh(Body_Mesh=GetSkeletalMeshForPartType(self=@n23))
   TryGetPawnOwner()
   Get bIsInFrontEnd()
  ]}
]
variables [
 Body_Mesh {type=SkeletalMeshComponent}
 DynHair {type=boolean}
 DynHairBackpack {type=boolean}
 DynHairCape {type=boolean}
 FE Idel {type=boolean}
 Gravity Override Params {
  type=GravityOverrideParamsStruct
  default {
   skydive_DivingDown {jointName=root, gravityOverride=<0 -400 150>, windAmplitude=<200 -500 400>, windFrequency=<1.5 7 8>, windOffset=<0 -600 400>}
   skydive_FloatingDown {gravityOverride=<0 -1800 -1000>, windAmplitude=<200 -1000 400>, windOffset=<0 0 0>}
   skydive_GoingUp {gravityOverride=<-50 -1000 500>, windAmplitude=<100 -150 200>, windFrequency=<6 14 16>, windOffset=<50 600 500>}
   skydive_Parachuting {gravityOverride=<0 -3000 -2500>, windAmplitude=<100 300 1000>, windFrequency=<1.5 5 8>}
  }
 }
 GravityOverride {type=Vector}
 head_up {type=number}
 isWearingBackpack {type=boolean}
 isWearingCape {type=boolean}
 isWearingPack {type=boolean}
 RBAN Params {
  type=RBANParamsStruct
  default {
   dBNO {simSpaceSettings {
     maxAngularAcceleration=1000
     maxAngularVelocity=3
     maxLinearAcceleration=300
     maxLinearVelocity=700
     velocityScaleZ=0
     externalLinearDragV=<0 0.5 0>
    }}
   dBNO_Carried {simSpaceSettings {
     maxAngularAcceleration=1000
     maxAngularVelocity=3
     maxLinearAcceleration=300
     maxLinearVelocity=700
     velocityScaleZ=0
     externalLinearDragV=<0 0.5 0>
    }}
   emoteOrMelee {simSpaceSettings {
     maxAngularAcceleration=1000
     maxAngularVelocity=3
     maxLinearAcceleration=300
     maxLinearVelocity=700
     velocityScaleZ=0
     externalLinearDragV=<0 0.5 0>
    }}
   front_End {simSpaceSettings {
     maxAngularAcceleration=1000
     maxAngularVelocity=3
     maxLinearAcceleration=300
     maxLinearVelocity=700
     velocityScaleZ=0
     externalLinearDragV=<0 0.5 0>
    }}
   onGround_Crouch_Moving {simSpaceSettings {
     maxAngularAcceleration=1000
     maxAngularVelocity=3
     maxLinearAcceleration=300
     maxLinearVelocity=700
     velocityScaleZ=0
     externalLinearDragV=<0 0.5 0>
    }}
   onGround_Falling {simSpaceSettings {
     maxAngularAcceleration=1000
     maxAngularVelocity=3
     maxLinearAcceleration=300
     maxLinearVelocity=700
     velocityScaleZ=0
     externalLinearDragV=<0 0.5 0>
    }}
   onGround_Moving {simSpaceSettings {
     maxAngularAcceleration=1000
     maxAngularVelocity=3
     maxLinearAcceleration=300
     maxLinearVelocity=700
     velocityScaleZ=0
     externalLinearDragV=<0 0.5 0>
    }}
   onGround_Standing {simSpaceSettings {
     maxAngularAcceleration=1000
     maxAngularVelocity=3
     maxLinearAcceleration=300
     maxLinearVelocity=700
     velocityScaleZ=0
     externalLinearDragV=<0 0.5 0>
    }}
   rail_Grinding {simSpaceSettings {
     maxAngularAcceleration=1000
     maxAngularVelocity=3
     maxLinearAcceleration=300
     maxLinearVelocity=700
     velocityScaleZ=0
     worldAlpha=0.1
     externalLinearDragV=<0 0.5 0>
    }}
   riding_Moving {simSpaceSettings {
     maxAngularAcceleration=1000
     maxAngularVelocity=3
     maxLinearAcceleration=300
     maxLinearVelocity=700
     velocityScaleZ=0
     externalLinearDragV=<0 0.5 0>
    }}
   riding_Moving_Targeting {simSpaceSettings {
     maxAngularAcceleration=1000
     maxAngularVelocity=3
     maxLinearAcceleration=300
     maxLinearVelocity=700
     velocityScaleZ=0
     externalLinearDragV=<0 0.5 0>
    }}
   riding_Standing {simSpaceSettings {
     maxAngularAcceleration=1000
     maxAngularVelocity=3
     maxLinearAcceleration=300
     maxLinearVelocity=700
     velocityScaleZ=0
     externalLinearDragV=<0 0.5 0>
    }}
   skydive_DivingDown {simSpaceSettings {
     maxAngularAcceleration=1000
     maxAngularVelocity=3
     maxLinearAcceleration=300
     maxLinearVelocity=700
     velocityScaleZ=0
     externalLinearDragV=<0 0.5 0>
    }}
   skydive_FloatingDown {
    componentLinearAccScale=<0 0.5 0>
    componentLinearVelScale=<0 0.5 0>
    simSpaceSettings {
     maxAngularAcceleration=1000
     maxAngularVelocity=3
     maxLinearAcceleration=300
     maxLinearVelocity=700
     velocityScaleZ=0
     externalLinearDragV=<0 0.5 0>
    }
   }
   skydive_GoingUp {simSpaceSettings {
     maxAngularAcceleration=1000
     maxAngularVelocity=3
     maxLinearAcceleration=300
     maxLinearVelocity=700
     velocityScaleZ=0
     externalLinearDragV=<0 0.5 0>
    }}
   skydive_Parachuting {
    componentLinearAccScale=<0 0.1 0.3>
    componentLinearVelScale=<0 0.1 0.3>
    simSpaceSettings {
     maxAngularAcceleration=1000
     maxAngularVelocity=3
     maxLinearAcceleration=300
     maxLinearVelocity=700
     velocityScaleZ=0
     externalLinearDragV=<0 0.5 0>
    }
   }
   skydive_Surfing {simSpaceSettings {
     maxAngularAcceleration=1000
     maxAngularVelocity=3
     maxLinearAcceleration=300
     maxLinearVelocity=700
     velocityScaleZ=0
     externalLinearDragV=<0 0.5 0>
    }}
   swimming {simSpaceSettings {
     maxAngularAcceleration=1000
     maxAngularVelocity=3
     maxLinearAcceleration=300
     maxLinearVelocity=700
     velocityScaleZ=0
     externalLinearDragV=<0 0.5 0>
    }}
   tactical_Sprint {simSpaceSettings {
     maxAngularAcceleration=1000
     maxAngularVelocity=3
     maxLinearAcceleration=300
     maxLinearVelocity=700
     velocityScaleZ=0
     externalLinearDragV=<0 0.5 0>
    }}
   ziplining {simSpaceSettings {
     maxAngularAcceleration=1000
     maxAngularVelocity=3
     maxLinearAcceleration=300
     maxLinearVelocity=700
     velocityScaleZ=0
     externalLinearDragV=<0 0.5 0>
    }}
  }
 }
 Skel Mesh {type=SkeletalMeshComponent}
 skydive {type=boolean}
 Tassle 01 Alpha {type=number}
 Tassle 02 Alpha {type=number}
 Tassle 03 Alpha {type=number}
 Tassle 04 Alpha {type=number}
 TesselRotator_01 {type=Rotator}
 TesselRotator_02 {type=Rotator}
 TesselRotator_03 {type=Rotator}
 TesselRotator_04 {type=Rotator}
]
